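# Material function that shows or hides elements of the Ancients telescope's astral map by stage.
# It outputs 1 when the telescope parameter collection's Alpha reaches StageValue times the Stage_Visible_At input, otherwise 0.
# That hard switch scales the incoming material's emissive, opacity and opacity mask.
# When hidden, world position offset also collapses vertices onto the object pivot, shrinking the mesh to a point.
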MaterialFunction /Game/Environments/Asteria/Props/Ancients/Ancients_Telescope_A/FX/Materials/MF_Ch4s3_AstralMap_LayerVis
graph {nodes [
  n1 = BreakMaterialAttributes!(Attr=FunctionInput(Preview=MakeMaterialAttributes!(
     BaseColor
     Metallic
     Specular
     Roughness
     Anisotropy
     EmissiveColor
     Opacity
     OpacityMask
     Normal
     Tangent
     WorldPositionOffset
     SubsurfaceColor
     ClearCoat
     ClearCoatRoughness
     AmbientOcclusion
     Refraction
     CustomizedUVs_0
     CustomizedUVs_1
     CustomizedUVs_2
     CustomizedUVs_3
     CustomizedUVs_4
     CustomizedUVs_5
     CustomizedUVs_6
     CustomizedUVs_7
     PixelDepthOffset
     ShadingModel
     Displacement
    )) {bUsePreviewValueAsDefault=true, inputName=InMaterial, inputType=FunctionInput_MaterialAttributes})
  n2 = FunctionInput(Preview=ScalarParameter() {group=Calendar, parameterName=Astral_Map_VisibleStage}) {
   bUsePreviewValueAsDefault=true
   description="0-7, for example a value of 1 will mean this element is visible from stage 1 onwards."
   inputName=Stage_Visible_At
   inputType=FunctionInput_Scalar
  }
  n3 = Constant() {r=1}
  n4 = If(
   A=CollectionParameter() {collection=/Game/Environments/Asteria/Props/Ancients/Ancients_Telescope_A/FX/MPC_Ch4s3_Telescope, parameterName=Alpha}
   B=Multiply(
    A=CollectionParameter() {collection=&MPC_Ch4s3_Telescope, parameterName=StageValue}
    B=@n2
   )
   "A > B"=@n3
   "A == B"=@n3
   "A < B"=Constant()
  )
  n5 = Subtract(
   A=MaterialFunctionCall() {materialFunction=&ObjectPivotPoint}.Object Pivot Location
   B=WorldPosition().XYZ
  )
  FunctionOutput(None=MakeMaterialAttributes(
    BaseColor=@n1.BaseColor
    Metallic=@n1.BaseColor
    Specular=@n1.BaseColor
    Roughness=@n1.BaseColor
    Anisotropy=@n1.BaseColor
    EmissiveColor=Multiply(A=@n1.EmissiveColor, B=@n4)
    Opacity=Multiply(A=@n4, B=@n1.Opacity)
    OpacityMask=Multiply(A=@n1.OpacityMask, B=@n4)
    Normal=@n1.BaseColor
    Tangent=@n1.BaseColor
    WorldPositionOffset=LinearInterpolate(A=@n5, B=@n1.WorldPositionOffset, Alpha=@n4)
   ))
 ]}
