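# Additive unlit material for Niagara mesh particles, drawing the Ancients telescope's astral projection light beam.
# A noise-distorted laser texture supplies core and falloff colors, scaled by time-of-day brightness.
# Opacity combines screen-aligned panning dust, optional quality-dependent depth fade, vertex red and telescope parameter-collection stage and opacity values.
# SplineThicken offsets vertices to vary width with camera distance, and MF_Ch4s3_AstralMap_LayerVis wraps the result for stage visibility.
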Material /Game/Environments/Asteria/Props/Ancients/Ancients_Telescope_A/FX/Materials/M_FX_AstralPrj_LightingBeam
graph {
 properties {
  bUseMaterialAttributes=true
  bUsedWithNiagaraMeshParticles=true
  blendMode=BLEND_Additive
  pixelDepthOffsetMode=PDOM_Legacy
  shadingModel=MSM_Unlit
  usedShadingModels=Unlit
 }
 nodes [
  n1 = VertexColor()
  LinearInterpolate(B=Power(Base=NamedRerouteUsage(), Exp=ScalarParameter() {parameterName=SmokeEndContrast}), Alpha=Saturate(None=OneMinus(None=@n1.R))) {constA=1, constB=0}
  n2 = MaterialFunctionCall(
   Value_Morning=ScalarParameter() {defaultValue=1, group=TOD, parameterName=Brightness_Morning}
   Value_Day=ScalarParameter() {defaultValue=1, group=TOD, parameterName=Brightness_Day}
   Value_Evening=ScalarParameter() {defaultValue=1, group=TOD, parameterName=Brightness_Evening}
   Value_Night=ScalarParameter() {defaultValue=1, group=TOD, parameterName=Brightness_Night}
  ) {materialFunction=&Scalar_Time_of_Day}
  n3 = VectorParameter() {parameterName=LazerColor, defaultValue=<rgba 0 0 0 1>}
  n4 = TextureCoordinate()
  n5 = TextureCoordinate()
  n6 = VectorParameter() {group=Noise, parameterName=Noise_Scale, defaultValue=<rgba 1 1 0 1>}
  n7 = ComponentMask(None=@n6.RGB) {g=true, r=true}
  n8 = TextureSampleParameter2D(UVs=Panner(Coordinate=Multiply(A=@n5, B=@n7), Speed=AppendVector(A=@n6.B, B=@n6.B))) {parameterName=MultiplyTexture1, texture=/Engine/EngineVolumetrics/LightBeam/Materials/T_EV_DustPanner_01}
  n9 = Subtract(A=Multiply(A=@n8.R) {constB=2})
  n10 = ComponentMask(None=VectorParameter() {group=Noise, parameterName=Noise_Distort, defaultValue=<rgba 0 0 0 1>}.RGB) {g=true, r=true}
  n11 = TextureSample(UVs=Add(A=MaterialFunctionCall(XY=@n4) {materialFunction=&Swizzle}.YX, B=Multiply(A=@n9, B=@n10))) {texture=/Game/Environments/Asteria/Props/Ancients/Ancients_Telescope_A/FX/Textures/T_FX_AstralPrj_Lazer}
  n12 = ComponentMask(None=Power(Base=@n11.G, Exp=ScalarParameter() {defaultValue=2, group=Falloff, parameterName=FalloffExp})) {g=true}
  n13 = SphereMask(A=ComponentMask(None=TextureCoordinate()) {r=true}, B=Constant() {r=0.5}) {attenuationRadius=0.5, hardnessPercent=90}
  n14 = Add(
   A=Multiply(A=Multiply(A=@n3.RGB, B=@n11.R), B=ScalarParameter() {defaultValue=1, parameterName=Brightness})
   B=Multiply(A=Multiply(A=@n12, B=@n13), B=VectorParameter() {parameterName=FalloffColor, defaultValue=<rgba 0 0 0 1>}.RGB)
  )
  n15 = Constant() {r=1}
  n16 = CollectionParameter() {collection=/Game/Environments/Asteria/Props/Ancients/Ancients_Telescope_A/FX/MPC_Ch4s3_Telescope, parameterName=Alpha}
  n17 = CollectionParameter() {collection=&MPC_Ch4s3_Telescope, parameterName=Stage2}
  n18 = ComponentMask(None=TextureCoordinate() {uTiling=2}) {r=true}
  n19 = VertexInterpolator(VS=Power(Base=OneMinus(None=@n18)) {constExponent=4})
  n20 = MaterialFunctionCall!(Center Point World Position=ActorPositionWS()) {materialFunction=/Engine/Functions/Engine_MaterialFunctions02/UVs/WorldSpaceAlignedScreenCoordinates}
  n21 = ScalarParameter() {parameterName=SmokeScale}
  n22 = TextureSampleParameter2D(UVs=Panner(Coordinate=Divide(A=@n20.Offset and Scaled UVS, B=@n21)) {speedY=0.05}) {parameterName=MultiplyTexture1, texture=&T_EV_DustPanner_01}
  n23 = DepthFade(Opacity=@n22.G)
  n24 = QualitySwitch(Default=@n23, Low=@n22.G, High=@n23, Medium=@n22.G, Epic=@n23)
  n25 = Multiply(
   A=LinearInterpolate(
    A=@n15
    B=If(A=@n16, B=@n17, "A > B"=@n15, "A == B"=@n15, "A < B"=@n19.PS) {constB=0.35}
    Alpha=ScalarParameter() {defaultValue=1, parameterName=FadeEndOverStages}
   )
   B=Multiply(A=StaticSwitchParameter(True=@n24, False=@n22.G) {parameterName=bUseDepthFade}, B=@n1.R)
  )
  n26 = MaterialFunctionCall(Float4=CollectionParameter() {collection=&MPC_Ch4s3_Telescope, parameterName=TOD_Opacity}) {materialFunction=&BreakOutFloat4Components}
  n27 = VertexInterpolator(VS=MaterialFunctionCall(Value_Morning=@n26.R, Value_Day=@n26.R, Value_Evening=@n26.R, Value_Night=@n26.R) {materialFunction=&Scalar_Time_of_Day}.Result)
  n28 = Distance(A=WorldPosition().XYZ, B=CameraPositionWS())
  n29 = ScalarParameter() {group=Spline, parameterName=Distance_Min}
  n30 = ScalarParameter() {group=Spline, parameterName=Distance_Max}
  n31 = Constant()
  n32 = Constant() {r=1}
  n33 = LinearInterpolate(
   A=ScalarParameter() {group=Spline, parameterName=NearThickness}
   B=ScalarParameter() {group=Spline, parameterName=FarThickness}
   Alpha=Clamp(None=MaterialFunctionCall(Input=@n28, Input Low=@n29, Input High=@n30, Target Low=@n31, Target High=@n32) {materialFunction=&RemapValueRange}.Result)
  )
  n34 = StaticBool()
  MaterialFunctionCall!(
   Stage_Visible_At
   InMaterial=MakeMaterialAttributes!(
    BaseColor
    Metallic
    Specular
    Roughness
    Anisotropy
    EmissiveColor=Multiply(A=@n2.Result, B=@n14)
    Opacity=Multiply(A=@n25, B=@n27.PS)
    OpacityMask
    Normal
    Tangent
    WorldPositionOffset=MaterialFunctionCall(WidthBase=@n33, WidthTip=@n33, Expand U or V UV Channel=@n34) {materialFunction=&SplineThicken}.WorldPositionOffset
    SubsurfaceColor
    ClearCoat
    ClearCoatRoughness
    AmbientOcclusion
    Refraction
    CustomizedUVs_0
    CustomizedUVs_1
    CustomizedUVs_2
    CustomizedUVs_3
    CustomizedUVs_4
    CustomizedUVs_5
    CustomizedUVs_6
    CustomizedUVs_7
    PixelDepthOffset
    ShadingModel
    Displacement
   )
  ) {materialFunction=&/MF_Ch4s3_AstralMap_LayerVis}
  NamedRerouteDeclaration(None=@n8.R) {name=SmokeLayer, nodeColor=<rgba 0.9219 0 1 1>}
  ObjectPositionWS()
  SmoothStep()
 ]
}
