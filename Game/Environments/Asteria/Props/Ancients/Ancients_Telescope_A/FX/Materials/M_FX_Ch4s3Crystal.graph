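# Unlit additive glow material for Niagara mesh particles, a crystal effect on the Ancients Telescope prop.
# Particle color tints a tight Fresnel rim, shaped by a crystal shard normal map, summed with a soft sphere-mask glow.
# The mask centers on the mesh particle pivot, raised by a ZOffset dynamic parameter and sized by LightRadius.
# A CameraOffset dynamic parameter feeds a camera-offset world position function.
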Material /Game/Environments/Asteria/Props/Ancients/Ancients_Telescope_A/FX/Materials/M_FX_Ch4s3Crystal
graph {
 properties {
  bUsedWithNiagaraMeshParticles=true
  blendMode=BLEND_Additive
  pixelDepthOffsetMode=PDOM_Legacy
  shadingModel=MSM_Unlit
  usedShadingModels=Unlit
 }
 nodes [
  n1 = ParticleColor()
  n2 = Fresnel(Normal=TextureSample() {samplerType=SAMPLERTYPE_Normal, texture=/Game/Environments/Apollo/Sets/DesertPoint/Crystals/Textures/Shards_CD_N}.RGB) {exponent=8}
  n3 = Constant()
  n4 = DynamicParameter() {paramNames [LightRadius,CameraOffset,ZOffset,Param4]}
  n5 = Add(
   A=MaterialFunctionCall() {materialFunction=&ObjectPivotPoint}.Mesh Particle Pivot Location
   B=AppendVector(A=AppendVector(A=@n3, B=@n3), B=@n4.ZOffset)
  )
  n6 = WorldPosition()
  Add(
   A=Multiply(A=@n1.RGB, B=Multiply(A=@n2) {constB=2})
   B=Multiply(A=@n1.RGB, B=SphereMask(A=@n5, B=@n6.XYZ, Radius=@n4.LightRadius) {hardnessPercent=0})
  )
  Constant3Vector()
  MaterialFunctionCall(Offset Amount=@n4.CameraOffset) {materialFunction=/Engine/Functions/Engine_MaterialFunctions02/WorldPositionOffset/CameraOffset}
  Transform()
 ]
}
