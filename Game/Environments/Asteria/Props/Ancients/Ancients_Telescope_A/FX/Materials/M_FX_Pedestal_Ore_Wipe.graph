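# Additive unlit Niagara mesh-particle material for a malachite ore effect on the Ancients telescope pedestal.
# Malachite diffuse green plus a normal-mapped Fresnel is tinted by particle color and fed through an emissive adjustment function.
# Dynamic parameters drive the Fresnel exponent and a CameraOffset amount.
# Particle alpha is multiplied by a bounding-box vertical (UVW blue) gradient, blended in by UseGradient.
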Material /Game/Environments/Asteria/Props/Ancients/Ancients_Telescope_A/FX/Materials/M_FX_Pedestal_Ore_Wipe
graph {
 properties {
  bUsedWithNiagaraMeshParticles=true
  blendMode=BLEND_Additive
  pixelDepthOffsetMode=PDOM_Legacy
  shadingModel=MSM_Unlit
  usedShadingModels=Unlit
 }
 nodes [
  ComponentMask(None=TextureCoordinate()) {g=true}
  n1 = TextureSample(UVs=TextureCoordinate()) {texture=/Game/Environments/Props/Ore/Textures/T_Props_Ore_Malachite_D}
  n2 = DynamicParameter() {paramNames [FresnelExp,Cam,Bright,Param4]}
  n3 = Fresnel(
   ExponentIn=@n2.FresnelExp
   Normal=TextureSample() {samplerType=SAMPLERTYPE_Normal, texture=/Game/Environments/Props/Ore/Textures/T_Props_Ore_Malachite_N}.RGB
  ) {baseReflectFraction=0}
  n4 = ParticleColor()
  MaterialFunctionCall(Emissive=Multiply(A=Add(A=@n1.G, B=@n3), B=@n4.RGB)) {materialFunction=&MF_FX_EmissiveAdjustments}
  MaterialFunctionCall(Offset Amount=@n2.Cam) {materialFunction=/Engine/Functions/Engine_MaterialFunctions02/WorldPositionOffset/CameraOffset}
  Multiply(
   A=@n4.A
   B=LinearInterpolate(
    A=Constant() {r=1}
    B=MaterialFunctionCall() {materialFunction=&BoundingBoxBased_0-1_UVW}.B
    Alpha=ScalarParameter() {defaultValue=1, parameterName=UseGradient}
   )
  )
 ]
}
