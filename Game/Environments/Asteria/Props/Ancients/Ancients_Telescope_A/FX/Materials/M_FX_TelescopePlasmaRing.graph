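# Additive unlit material for Niagara mesh particles, drawing a plasma ring effect for the Ancients telescope in Asteria.
# SplineThicken and CameraOffset build world position offset from dynamic-parameter width and camera offset.
# A panned plasma texture distorts a UV axis feeding a powered SphereMask, forming a wavy band scaled by particle alpha.
# Particle color feeds emissive adjustments; a quality switch skips depth fade on Low and Medium.
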Material /Game/Environments/Asteria/Props/Ancients/Ancients_Telescope_A/FX/Materials/M_FX_TelescopePlasmaRing
graph {
 properties {
  bUsedWithNiagaraMeshParticles=true
  blendMode=BLEND_Additive
  pixelDepthOffsetMode=PDOM_Legacy
  shadingModel=MSM_Unlit
  usedShadingModels=Unlit
 }
 nodes [
  n1 = ParticleColor()
  MaterialFunctionCall(Emissive=@n1.RGB) {materialFunction=&MF_FX_EmissiveAdjustments}
  n2 = DynamicParameter() {defaultValue=<rgba 75 1 1 1>, paramNames [None,Width,Param3,CamOffset]}
  n3 = MaterialFunctionCall(WidthBase=@n2.Width, WidthTip=@n2.Width) {materialFunction=&SplineThicken}
  n4 = Add(A=@n3.WorldPositionOffset, B=WorldPosition().XYZ)
  Multiply(
   A=Add(
    A=@n3.WorldPositionOffset
    B=MaterialFunctionCall(Offset Amount=@n2.CamOffset, World Position=@n4) {materialFunction=/Engine/Functions/Engine_MaterialFunctions02/WorldPositionOffset/CameraOffset}.Camera Offset
   )
   B=ScalarParameter() {defaultValue=1, parameterName=WPO Strength}
  )
  n5 = VectorParameter() {parameterName=PlasMa Tex Option, channelNames {a=Speed Y, b=Speed X, g=Tile Y, r=tile X}, defaultValue=<rgba 0.5 2 1 0>}
  n6 = Multiply(A=TextureCoordinate(), B=AppendVector(A=@n5.tile X, B=@n5.tile X))
  n7 = AppendVector(A=@n5.Speed X, B=@n5.Speed X)
  n8 = TextureSampleParameter2D(UVs=LinearInterpolate(
    A=Panner(Coordinate=@n6, Speed=@n7) {speedX=1}
    B=Panner(Coordinate=@n6, Speed=@n7) {bFractionalPart=true, speedX=1}
    Alpha=ScalarParameter() {parameterName=use fractional Panner}
   )) {
   parameterName=Plasma Tex
   samplerType=SAMPLERTYPE_LinearGrayscale
   texture=/Game/Effects/Fort_Effects/Emotes/UltraLight/Texture/T_Plasma_C_Greyscale
  }
  n9 = Multiply(A=@n8.G) {constB=2}
  n10 = TextureCoordinate()
  n11 = Add(
   A=Multiply(A=Subtract(A=@n9), B=ScalarParameter() {defaultValue=0.15, parameterName=Distort Str}) {constB=0.15}
   B=LinearInterpolate(A=ComponentMask(None=@n10) {r=true}, B=ComponentMask(None=@n10) {g=true}, Alpha=ScalarParameter() {parameterName=Flip Sphere mask Direction})
  )
  n12 = ScalarParameter() {defaultValue=0.5, parameterName=Sphere mask Center Point}
  n13 = ScalarParameter() {defaultValue=0.3, parameterName=Sphere mask Radius}
  n14 = ScalarParameter() {parameterName=Sphere mask Hardness}
  n15 = Multiply(
   A=@n1.A
   B=Power(
    Base=SphereMask(A=@n11, B=@n12, Radius=@n13, Hardness=@n14) {attenuationRadius=0.3, hardnessPercent=0}
    Exp=ScalarParameter() {defaultValue=4, parameterName=Sphere mask Power}
   ) {constExponent=4}
  )
  n16 = DepthFade(Opacity=@n15, FadeDistance=ScalarParameter() {defaultValue=100, parameterName=Fade Distacne})
  QualitySwitch(Default=@n16, Low=@n15, High=@n16, Medium=@n15, Epic=@n16)
 ]
}
