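# Additive unlit Niagara sprite material for energy fragment particles on the Ancients telescope prop in Asteria.
# Vertically panned Perlin noise, remapped and scaled down, distorts the UVs of an energy streak texture, with per-particle random offsets varying each sprite.
# A SmoothStep driven by the "erode" dynamic parameter dissolves the streaks.
# A soft centred SphereMask fades the sprite edges, and particle alpha scales the result.
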Material /Game/Environments/Asteria/Props/Ancients/Ancients_Telescope_A/FX/Materials/M_FX_Telescope_EnergyFrags
graph {
 properties {bUsedWithNiagaraSprites=true, blendMode=BLEND_Additive, pixelDepthOffsetMode=PDOM_Legacy, shadingModel=MSM_Unlit, usedShadingModels=Unlit}
 nodes [
  Constant()
  n1 = DynamicParameter() {paramNames [erode,Param2,Param3,Param4]}
  n2 = Add(A=@n1.erode) {constB=3}
  n3 = Panner(Coordinate=Add(A=TextureCoordinate() {uTiling=2, vTiling=0.4}, B=ParticleRandom())) {speedY=-0.5}
  n4 = Subtract(A=Multiply(A=TextureSample(UVs=@n3) {samplerType=SAMPLERTYPE_LinearGrayscale, texture=&T_SimplePerlin_Greyscale}.G) {constB=2})
  n5 = TextureCoordinate() {uTiling=1.75}
  n6 = ParticleRandom()
  n7 = TextureSample(UVs=Add(A=Multiply(A=@n4) {constB=0.1}, B=Add(A=@n5, B=@n6))) {texture=/Game/Effects/Fort_Effects/Textures/Energy/T_OutpostShockwaveStreaks_02}
  n8 = TextureCoordinate()
  n9 = Constant() {r=0.5}
  Multiply(
   A=ParticleColor().A
   B=Multiply(A=SmoothStep(Min=@n1.erode, Max=@n2, Value=@n7.R), B=SphereMask(A=@n8, B=@n9) {attenuationRadius=0.5, hardnessPercent=0})
  )
  OneMinus(None=@n1.erode)
  Power(Base=@n7.R)
  ScalarParameter() {parameterName=Point}
  TextureSample()
 ]
}
